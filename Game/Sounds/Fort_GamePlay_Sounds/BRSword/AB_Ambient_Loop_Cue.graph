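# Sound cue for the looping ambient sound of the AB sword weapon in BRSword, reusing the jump attack loop wave at lowered pitch.
# The looping wave passes through a volume fade-in envelope, a modulator, attenuation, a Fort_Player_Foley class override and a mixer.
# Attenuation overrides add custom falloff curves, distance low-pass filtering and listener focus.
# A concurrency override caps simultaneous instances, stopping the quietest.
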SoundCue /Game/Sounds/Fort_GamePlay_Sounds/BRSword/AB_Ambient_Loop_Cue
root=Mixer_1
nodes [
 {
  id=Attenuation_1
  children [Modulator_0]
  properties {
   attenuationSettings=/Game/Sounds/Attenuation/Fort_Weapon_AB_Ambient_Athena_Att
   bOverrideAttenuation=true
   attenuationOverrides {
    bAttenuateWithLPF=true
    bEnableListenerFocus=true
    bEnableLogFrequencyScaling=true
    distanceAlgorithm=Custom
    falloffDistance=10000
    focusAttackInterpSpeed=10
    focusAzimuth=20
    focusReleaseInterpSpeed=20
    lPFFrequencyAtMax=2000
    lPFRadiusMax=5000
    lPFRadiusMin=0
    nonFocusAzimuth=40
    nonFocusVolumeAttenuation=0.4
    nonSpatializedRadiusStart=500
    stereoSpread=3000
    attenuationShapeExtents {x=0}
    customAttenuationCurve {editorCurveData {keys [
       {arriveTangent=-1.8367, interpMode=RCIM_Cubic, leaveTangent=-1.8367, tangentMode=RCTM_User, time=-0.003, value=1.0617}
       <at 0.1247 0.0314>
       {arriveTangent=-0.1319, interpMode=RCIM_Cubic, leaveTangent=-0.1319, tangentMode=RCTM_User, time=0.5, value=-0.037}
      ]}}
    customLowpassAirAbsorptionCurve {editorCurveData {keys [<at -0.0065 1.006>,<at 0.1447 0.126>,<at 0.9945 -0.0104>]}}
   }
  }
 }
 {
  id=Enveloper_0
  children [WavePlayer_0]
  properties {pitchCurve {editorCurveData {keys [{time=0}]}}, volumeCurve {editorCurveData {keys [<at 0.0117 0.0072>,<at 0.5073 1>]}}}
 }
 {id=Mixer_1, children [SoundClass_1], properties {inputVolume [0.9]}}
 {id=Modulator_0, children [Enveloper_0]}
 {id=SoundClass_1, children [Attenuation_1], properties {soundClassOverride=&Fort_Player_Foley}}
 {id=WavePlayer_0, properties {bLooping=true, soundWaveAssetPtr=&/AB_Jump_Attack_Loop_01}}
]
properties {
 bOverrideConcurrency=true
 maxDistance=4999.7983
 pitchMultiplier=0.85
 volumeMultiplier=1.1
 concurrencyOverrides {maxCount=3, resolutionRule=StopQuietest}
}
