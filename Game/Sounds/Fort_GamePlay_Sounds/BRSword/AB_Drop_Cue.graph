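# Sound cue for a Battle Royale sword drop, built from close and distant versions of a reused swing recording.
# A mixer blends the two layers, each with random pitch modulation, its own close or far Carmine weapon attenuation, and a third-person ranged-fire sound class.
# The cue raises overall pitch, uses weapon-drop attenuation, and allows one instance at a time, stopping the oldest.
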SoundCue /Game/Sounds/Fort_GamePlay_Sounds/BRSword/AB_Drop_Cue
root=Mixer_0
nodes [
 {
  id=Attenuation_1
  children [Modulator_0]
  properties {attenuationSettings=/Game/Sounds/Attenuation/Fort_Weapon_Carmine_Close_Athena_Att, attenuationOverrides {bSpatialize=false}}
 }
 {
  id=Attenuation_2
  children [Modulator_1]
  properties {attenuationSettings=/Game/Sounds/Attenuation/Fort_Weapon_Carmine_Far_Athena_Att, attenuationOverrides {bSpatialize=false}}
 }
 {id=Mixer_0, children [SoundClass_0,SoundClass_1], properties {inputVolume [0.75,1]}}
 {id=Modulator_0, children [WavePlayer_18], properties {pitchMax=1.1, pitchMin=0.8}}
 {id=Modulator_1, children [WavePlayer_21], properties {pitchMax=1.1, pitchMin=0.8}}
 {
  id=SoundClass_0
  children [Attenuation_1]
  properties {soundClassOverride=&Fort_WeaponRanged_Fire_3P_PSM}
 }
 {
  id=SoundClass_1
  children [Attenuation_2]
  properties {soundClassOverride=&Fort_WeaponRanged_Fire_3P_PSM}
 }
 {id=WavePlayer_18, properties {soundWaveAssetPtr=&/AB_Swing_Second_03}}
 {id=WavePlayer_21, properties {soundWaveAssetPtr=&/AB_Swing_Second_03_distant}}
]
properties {
 attenuationSettings=&Fort_Weapon_Drop_Att
 bOverrideConcurrency=true
 maxDistance=25000
 pitchMultiplier=1.5
 volumeMultiplier=0.9
 concurrencyOverrides {maxCount=1, resolutionRule=StopOldest}
}
