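# Sound cue for the BRSword AB impact, layering close and distant versions of the hit.
# Each layer picks one of three waves at random, applies upward pitch variation, then passes through its own Carmine close or far attenuation and a third-person ranged-weapon-fire sound class override.
# A mixer combines both layers, with the close one slightly lowered.
# Concurrency is overridden to a small per-owner limit that stops the oldest instance.
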SoundCue /Game/Sounds/Fort_GamePlay_Sounds/BRSword/AB_Impact_Cue
root=Mixer_1
nodes [
 {id=Attenuation_1, children [Modulator_0], properties {attenuationSettings=/Game/Sounds/Attenuation/Fort_Weapon_Carmine_Close_Athena_Att}}
 {id=Attenuation_2, children [Modulator_1], properties {attenuationSettings=/Game/Sounds/Attenuation/Fort_Weapon_Carmine_Far_Athena_Att}}
 {id=Mixer_1, children [SoundClass_1,SoundClass_2], properties {inputVolume [0.9,1]}}
 {id=Modulator_0, children [Random_4], properties {pitchMax=1.3, pitchMin=1}}
 {id=Modulator_1, children [Random_5], properties {pitchMax=1.3, pitchMin=1}}
 {id=Random_4, children [WavePlayer_8, WavePlayer_10, WavePlayer_11], properties {weights [1,1,1]}}
 {id=Random_5, children [WavePlayer_12, WavePlayer_14, WavePlayer_15], properties {weights [1,1,1]}}
 {
  id=SoundClass_1
  children [Attenuation_1]
  properties {soundClassOverride=&Fort_WeaponRanged_Fire_3P_PSM}
 }
 {
  id=SoundClass_2
  children [Attenuation_2]
  properties {soundClassOverride=&Fort_WeaponRanged_Fire_3P_PSM}
 }
 {id=WavePlayer_10, properties {soundWaveAssetPtr=&/AB_Impact_03}}
 {id=WavePlayer_11, properties {soundWaveAssetPtr=&/AB_Impact_04}}
 {id=WavePlayer_12, properties {soundWaveAssetPtr=&/AB_Impact_01_distant}}
 {id=WavePlayer_14, properties {soundWaveAssetPtr=&/AB_Impact_03_distant}}
 {id=WavePlayer_15, properties {soundWaveAssetPtr=&/AB_Impact_04_distant}}
 {id=WavePlayer_8, properties {soundWaveAssetPtr=&/AB_Impact_01}}
]
properties {
 bOverrideConcurrency=true
 maxDistance=25000
 volumeMultiplier=1
 concurrencyOverrides {bLimitToOwner=true, maxCount=2, resolutionRule=StopOldest}
}
