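# Looping sound cue in the BRSword sound set, named as an item loop and assigned to the loot reward sound class.
# A single looping wave player feeds the root, reusing the AB_Jump_Attack_Loop_01 wave at reduced volume.
# The cue overrides attenuation with logarithmic falloff, stereo spread and listener focus, so it plays quieter when out of focus.
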SoundCue /Game/Sounds/Fort_GamePlay_Sounds/BRSword/AB_ItemLoop_Cue
root=WavePlayer_1
nodes [{id=WavePlayer_1, properties {bLooping=true, soundWaveAssetPtr=&/AB_Jump_Attack_Loop_01}}]
properties {
 bOverrideAttenuation=true
 maxDistance=1681.9642
 soundClassObject=&Fort_Loot_Reward_PSM
 volumeMultiplier=0.3
 attenuationOverrides {
  bEnableListenerFocus=true
  distanceAlgorithm=Logarithmic
  falloffDistance=681.9642
  nonFocusVolumeAttenuation=0.4762
  stereoSpread=35.6359
  attenuationShapeExtents {x=1000}
 }
}
