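# Sound cue for the impact of the AB jump attack in the BRSword sound set, mixing close and distant versions at half volume each.
# Each branch has its own weapon attenuation and sound class override for distance-based playback.
# Both layer a random jump-attack impact with a random generic impact pitched down, under a shared upward pitch variation.
# Concurrency allows one instance per owner, stopping the oldest.
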SoundCue /Game/Sounds/Fort_GamePlay_Sounds/BRSword/AB_Jump_Attack_Impact_Cue
root=Mixer_1
nodes [
 {id=Attenuation_1, children [Modulator_0], properties {attenuationSettings=/Game/Sounds/Attenuation/Fort_Weapon_Carmine_Close_Athena_Att}}
 {id=Attenuation_2, children [Modulator_1], properties {attenuationSettings=/Game/Sounds/Attenuation/Fort_Weapon_Carmine_Far_Athena_Att}}
 {id=Mixer_0, children [Random_0,Modulator_2], properties {inputVolume [1,1]}}
 {id=Mixer_1, children [SoundClass_1,SoundClass_2], properties {inputVolume [0.5,0.5]}}
 {id=Mixer_2, children [Random_2,Modulator_3], properties {inputVolume [1,1]}}
 {id=Modulator_0, children [Mixer_0], properties {pitchMax=1.2, pitchMin=1}}
 {id=Modulator_1, children [Mixer_2], properties {pitchMax=1.2, pitchMin=1}}
 {id=Modulator_2, children [Random_1], properties {pitchMax=0.8, pitchMin=0.7}}
 {id=Modulator_3, children [Random_3], properties {pitchMax=0.8, pitchMin=0.7}}
 {id=Random_0, children [WavePlayer_0,WavePlayer_1,WavePlayer_2], properties {weights [1,1,1]}}
 {id=Random_1, children [WavePlayer_3,WavePlayer_4,WavePlayer_5,WavePlayer_6], properties {weights [1,1,1,1]}}
 {id=Random_2, children [WavePlayer_7, WavePlayer_16, WavePlayer_17], properties {weights [1,1,1]}}
 {id=Random_3, children [WavePlayer_18, WavePlayer_19, WavePlayer_20, WavePlayer_21], properties {weights [1,1,1,1]}}
 {
  id=SoundClass_1
  children [Attenuation_1]
  properties {soundClassOverride=&Fort_WeaponRanged_Fire_3P_PSM}
 }
 {
  id=SoundClass_2
  children [Attenuation_2]
  properties {soundClassOverride=&Fort_WeaponRanged_Fire_Distant}
 }
 {id=WavePlayer_0, properties {soundWaveAssetPtr=&/AB_Jump_Attack_Impact_01}}
 {id=WavePlayer_1, properties {soundWaveAssetPtr=&/AB_Jump_Attack_Impact_02}}
 {id=WavePlayer_16, properties {soundWaveAssetPtr=&/AB_Jump_Attack_Impact_02_distant}}
 {id=WavePlayer_17, properties {soundWaveAssetPtr=&/AB_Jump_Attack_Impact_03_distant}}
 {id=WavePlayer_18, properties {soundWaveAssetPtr=&/AB_Impact_01_distant}}
 {id=WavePlayer_19, properties {soundWaveAssetPtr=&/AB_Impact_02_distant}}
 {id=WavePlayer_2, properties {soundWaveAssetPtr=&/AB_Jump_Attack_Impact_03}}
 {id=WavePlayer_20, properties {soundWaveAssetPtr=&/AB_Impact_03_distant}}
 {id=WavePlayer_21, properties {soundWaveAssetPtr=&/AB_Impact_04_distant}}
 {id=WavePlayer_3, properties {soundWaveAssetPtr=&/AB_Impact_01}}
 {id=WavePlayer_4, properties {soundWaveAssetPtr=&/AB_Impact_02}}
 {id=WavePlayer_5, properties {soundWaveAssetPtr=&/AB_Impact_03}}
 {id=WavePlayer_6, properties {soundWaveAssetPtr=&/AB_Impact_04}}
 {id=WavePlayer_7, properties {soundWaveAssetPtr=&/AB_Jump_Attack_Impact_01_distant}}
]
properties {
 bOverrideConcurrency=true
 maxDistance=25000
 volumeMultiplier=1
 concurrencyOverrides {bLimitToOwner=true, maxCount=1, resolutionRule=StopOldest}
}
