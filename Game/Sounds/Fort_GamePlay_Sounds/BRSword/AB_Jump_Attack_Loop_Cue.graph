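# Sound cue for the BR sword jump attack, combining a one-shot start, a looping body and a swing in close and distant versions.
# Each version mixes a random start, a loop faded in by an enveloper and a random pitched-down swing, under slight pitch variation.
# The versions play together through separate close and far attenuations and third-person and distant fire sound classes.
# Concurrency keeps one instance per owner, stopping the oldest.
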SoundCue /Game/Sounds/Fort_GamePlay_Sounds/BRSword/AB_Jump_Attack_Loop_Cue
root=Mixer_1
nodes [
 {id=Attenuation_1, children [Modulator_0], properties {attenuationSettings=/Game/Sounds/Attenuation/Fort_Weapon_Carmine_Close_Athena_Att}}
 {id=Attenuation_2, children [Modulator_1], properties {attenuationSettings=/Game/Sounds/Attenuation/Fort_Weapon_Carmine_Far_Athena_Att}}
 {
  id=Enveloper_0
  children [WavePlayer_2]
  properties {pitchCurve {editorCurveData {keys [{time=0}]}}, volumeCurve {editorCurveData {keys [<at 0.0191 0.0138>,<at 0.5042 0.9972>]}}}
 }
 {
  id=Enveloper_1
  children [WavePlayer_16]
  properties {pitchCurve {editorCurveData {keys [{time=0}]}}, volumeCurve {editorCurveData {keys [<at 0.0191 0.0138>,<at 0.5042 0.9972>]}}}
 }
 {id=Mixer_0, children [Random_0,Enveloper_0,Modulator_2], properties {inputVolume [1,1,1]}}
 {id=Mixer_1, children [SoundClass_1,SoundClass_2], properties {inputVolume [0.5,0.5]}}
 {id=Mixer_2, children [Random_2,Enveloper_1,Modulator_3], properties {inputVolume [1,1,1]}}
 {id=Modulator_0, children [Mixer_0], properties {pitchMax=1.2, pitchMin=1}}
 {id=Modulator_1, children [Mixer_2], properties {pitchMax=1.2, pitchMin=1}}
 {id=Modulator_2, children [Random_1], properties {pitchMax=0.85, pitchMin=0.75}}
 {id=Modulator_3, children [Random_3], properties {pitchMax=0.85, pitchMin=0.75}}
 {id=Random_0, children [WavePlayer_0,WavePlayer_1], properties {weights [1,1]}}
 {id=Random_1, children [WavePlayer_3,WavePlayer_4,WavePlayer_5], properties {weights [1,1,1]}}
 {id=Random_2, children [WavePlayer_6,WavePlayer_7], properties {weights [1,1]}}
 {id=Random_3, children [WavePlayer_17, WavePlayer_18, WavePlayer_19], properties {weights [1,1,1]}}
 {
  id=SoundClass_1
  children [Attenuation_1]
  properties {soundClassOverride=&Fort_WeaponRanged_Fire_3P_PSM}
 }
 {
  id=SoundClass_2
  children [Attenuation_2]
  properties {soundClassOverride=&Fort_WeaponRanged_Fire_Distant}
 }
 {id=WavePlayer_0, properties {soundWaveAssetPtr=&/AB_Jump_Attack_Start_01}}
 {id=WavePlayer_1, properties {soundWaveAssetPtr=&/AB_Jump_Attack_Start_02}}
 {id=WavePlayer_16, properties {bLooping=true, soundWaveAssetPtr=&/AB_Jump_Attack_Loop_01_distant}}
 {id=WavePlayer_17, properties {soundWaveAssetPtr=&/AB_Swing_Second_01_distant}}
 {id=WavePlayer_18, properties {soundWaveAssetPtr=&/AB_Swing_Second_02_distant}}
 {id=WavePlayer_19, properties {soundWaveAssetPtr=&/AB_Swing_Second_03_distant}}
 {id=WavePlayer_2, properties {bLooping=true, soundWaveAssetPtr=&/AB_Jump_Attack_Loop_01}}
 {id=WavePlayer_3, properties {soundWaveAssetPtr=&/AB_Swing_Second_01}}
 {id=WavePlayer_4, properties {soundWaveAssetPtr=&/AB_Swing_Second_02}}
 {id=WavePlayer_5, properties {soundWaveAssetPtr=&/AB_Swing_Second_03}}
 {id=WavePlayer_6, properties {soundWaveAssetPtr=&/AB_Jump_Attack_Start_01_distant}}
 {id=WavePlayer_7, properties {soundWaveAssetPtr=&/AB_Jump_Attack_Start_02_distant}}
]
properties {
 bOverrideConcurrency=true
 maxDistance=25000
 volumeMultiplier=1
 concurrencyOverrides {bLimitToOwner=true, maxCount=1, resolutionRule=StopOldest}
}
