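# Sound cue for the successful end of a BR sword pull.
# Parallel close and distant branches use matching wave sets with separate Carmine attenuations, both under the third-person ranged weapon fire sound class.
# Each branch opens with a pitched-down whoosh, then after a delay layers an impact, success sting, second whoosh and swing at staggered offsets.
# Concurrency allows one instance and rejects new plays while it sounds.
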SoundCue /Game/Sounds/Fort_GamePlay_Sounds/BRSword/AB_Pull_End_Success_Cue
root=Mixer_1
nodes [
 {id=Attenuation_1, children [Modulator_0], properties {attenuationSettings=/Game/Sounds/Attenuation/Fort_Weapon_Carmine_Close_Athena_Att}}
 {id=Attenuation_2, children [Modulator_1], properties {attenuationSettings=/Game/Sounds/Attenuation/Fort_Weapon_Carmine_Far_Athena_Att}}
 {id=Delay_1, children [Modulator_5], properties {delayMax=0.4, delayMin=0.4}}
 {id=Delay_10, children [Modulator_8], properties {delayMax=0.4, delayMin=0.4}}
 {id=Delay_11, children [Mixer_5], properties {delayMax=0.7, delayMin=0.7}}
 {id=Delay_2, children [WavePlayer_0], properties {delayMax=0.05, delayMin=0.05}}
 {id=Delay_5, children [WavePlayer_7], properties {delayMax=0.65, delayMin=0.65}}
 {id=Delay_7, children [Mixer_2], properties {delayMax=0.7, delayMin=0.7}}
 {id=Delay_8, children [WavePlayer_9], properties {delayMax=0.05, delayMin=0.05}}
 {id=Delay_9, children [WavePlayer_11], properties {delayMax=0.65, delayMin=0.65}}
 {id=Mixer_1, children [SoundClass_1,SoundClass_2], properties {inputVolume [0.7,1]}}
 {id=Mixer_2, children [Delay_2,Modulator_2,Delay_5,Delay_1], properties {inputVolume [1.2,1,1,1]}}
 {id=Mixer_4, children [Delay_7,Modulator_3], properties {inputVolume [1,0.7]}}
 {id=Mixer_5, children [Delay_8,Modulator_6,Delay_9,Delay_10], properties {inputVolume [1.2,1,1,1]}}
 {id=Mixer_6, children [Delay_11,Modulator_7], properties {inputVolume [1,0.7]}}
 {id=Modulator_0, children [Mixer_4], properties {pitchMin=1, volumeMin=1}}
 {id=Modulator_1, children [Mixer_6], properties {pitchMin=1}}
 {id=Modulator_2, children [WavePlayer_2], properties {pitchMax=0.9, pitchMin=0.9}}
 {id=Modulator_3, children [WavePlayer_8], properties {pitchMax=0.9, pitchMin=0.8}}
 {id=Modulator_5, children [WavePlayer_4], properties {pitchMax=0.8, pitchMin=0.8, volumeMax=1, volumeMin=1}}
 {id=Modulator_6, children [WavePlayer_10], properties {pitchMax=0.9, pitchMin=0.9}}
 {id=Modulator_7, children [WavePlayer_13], properties {pitchMax=0.9, pitchMin=0.8}}
 {id=Modulator_8, children [WavePlayer_12], properties {pitchMax=0.8, pitchMin=0.8, volumeMax=1, volumeMin=1}}
 {
  id=SoundClass_1
  children [Attenuation_1]
  properties {soundClassOverride=&Fort_WeaponRanged_Fire_3P_PSM}
 }
 {
  id=SoundClass_2
  children [Attenuation_2]
  properties {soundClassOverride=&Fort_WeaponRanged_Fire_3P_PSM}
 }
 {id=WavePlayer_0, properties {soundWaveAssetPtr=&/AB_Pull_End_Success_01}}
 {id=WavePlayer_10, properties {soundWaveAssetPtr=&/AB_Impact_04_distant}}
 {id=WavePlayer_11, properties {soundWaveAssetPtr=&/AB_Swing_Second_02_distant}}
 {id=WavePlayer_12, properties {soundWaveAssetPtr=&/AB_Pull_Whoosh_01_distant}}
 {id=WavePlayer_13, properties {soundWaveAssetPtr=&/AB_Pull_Whoosh_02_distant}}
 {id=WavePlayer_2, properties {soundWaveAssetPtr=&/AB_Impact_04}}
 {id=WavePlayer_4, properties {soundWaveAssetPtr=&/AB_Pull_Whoosh_01}}
 {id=WavePlayer_7, properties {soundWaveAssetPtr=&/AB_Swing_Second_02}}
 {id=WavePlayer_8, properties {soundWaveAssetPtr=&/AB_Pull_Whoosh_02}}
 {id=WavePlayer_9, properties {soundWaveAssetPtr=&/AB_Pull_End_Success_01_distant}}
]
properties {bOverrideConcurrency=true, maxDistance=25000, volumeMultiplier=1, concurrencyOverrides {maxCount=1, resolutionRule=PreventNew}}
