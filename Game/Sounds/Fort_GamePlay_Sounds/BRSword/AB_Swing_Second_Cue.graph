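# Sound cue for the second AB swing of the BRSword weapon, layering close and distant versions of the swing.
# Each layer randomly picks one of three equally weighted waves, varies its pitch, then applies its own close or far Athena attenuation.
# Both layers pass through a third-person ranged-weapon-fire sound class override into a mixer, with the close layer slightly quieter.
# A concurrency override caps copies per owner, stopping the oldest.
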SoundCue /Game/Sounds/Fort_GamePlay_Sounds/BRSword/AB_Swing_Second_Cue
root=Mixer_0
nodes [
 {
  id=Attenuation_1
  children [Modulator_0]
  properties {attenuationSettings=/Game/Sounds/Attenuation/Fort_Weapon_AB_Close_Athena_Att, attenuationOverrides {bSpatialize=false}}
 }
 {
  id=Attenuation_2
  children [Modulator_1]
  properties {attenuationSettings=/Game/Sounds/Attenuation/Fort_Weapon_AB_Far_Athena_Att, attenuationOverrides {bSpatialize=false}}
 }
 {id=Mixer_0, children [SoundClass_0,SoundClass_1], properties {inputVolume [0.8,1]}}
 {id=Modulator_0, children [Random_4], properties {pitchMax=1.1, pitchMin=0.8}}
 {id=Modulator_1, children [Random_5], properties {pitchMax=1.1, pitchMin=0.8}}
 {id=Random_4, children [WavePlayer_16, WavePlayer_17, WavePlayer_18], properties {weights [1,1,1]}}
 {id=Random_5, children [WavePlayer_19, WavePlayer_20, WavePlayer_21], properties {weights [1,1,1]}}
 {
  id=SoundClass_0
  children [Attenuation_1]
  properties {soundClassOverride=&Fort_WeaponRanged_Fire_3P_PSM}
 }
 {
  id=SoundClass_1
  children [Attenuation_2]
  properties {soundClassOverride=&Fort_WeaponRanged_Fire_3P_PSM}
 }
 {id=WavePlayer_16, properties {soundWaveAssetPtr=&/AB_Swing_Second_01}}
 {id=WavePlayer_17, properties {soundWaveAssetPtr=&/AB_Swing_Second_02}}
 {id=WavePlayer_18, properties {soundWaveAssetPtr=&/AB_Swing_Second_03}}
 {id=WavePlayer_19, properties {soundWaveAssetPtr=&/AB_Swing_Second_01_distant}}
 {id=WavePlayer_20, properties {soundWaveAssetPtr=&/AB_Swing_Second_02_distant}}
 {id=WavePlayer_21, properties {soundWaveAssetPtr=&/AB_Swing_Second_03_distant}}
]
properties {
 bOverrideConcurrency=true
 maxDistance=15000
 volumeMultiplier=1.15
 concurrencyOverrides {bLimitToOwner=true, maxCount=2, resolutionRule=StopOldest}
}
